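# Unlit, two-sided translucent material for iridescent digital confetti on Niagara sprite and mesh particles.
# An optional bi-fold mode uses vertex color red to rotate and tint two halves separately, animating a time-driven fold.
# View angle, panning UVs, per-particle random and time drive hue shift and brightness, optionally blended toward particle color.
# A view-space normal matcap lookup supplies a fake cartoon glass reflection mask alongside emissive adjustments.
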Material /SparksEffects/Effects/Materials/Digital/M_FX_TechConfetti
graph {
 properties {
  bAutomaticallySetUsageInEditor=false
  bUsedWithNiagaraMeshParticles=true
  bUsedWithNiagaraSprites=true
  blendMode=BLEND_Translucent
  shadingModel=MSM_Unlit
  twoSided=true
  usedShadingModels=Unlit
 }
 nodes [
  DynamicParameter() {paramNames [BiFold_Size,Param2,Param3,Param4]}
  n1 = Time()
  n2 = ScalarParameter() {defaultValue=0.2, group=BiFold, parameterName=BiFold_RotationSpeed}
  n3 = Multiply(A=@n1, B=@n2)
  n4 = Transform(None=Constant3Vector() {constant=<rgba 0 1 0 0>}) {transformSourceType=TRANSFORMSOURCE_Local}
  n5 = TransformPosition(None=Constant3Vector()) {transformType=TRANSFORMPOSSOURCE_World}
  n6 = WorldPosition() {worldPositionShaderOffset=WPT_ExcludeAllShaderOffsets}
  n7 = MaterialFunctionCall() {materialFunction=&ObjectPivotPoint}
  n8 = ScalarParameter() {defaultValue=1, parameterName=Scale}
  n9 = Add(
   A=WorldPosition() {worldPositionShaderOffset=WPT_ExcludeAllShaderOffsets}.XYZ
   B=Multiply(A=Subtract(A=@n6.XYZ, B=@n7.Object Pivot Location), B=Subtract(A=@n8))
  )
  n10 = MaterialFunctionCall!(
   RotationAngle=Multiply(A=Add(A=@n3) {constB=0.6}) {constB=-0.33}
   NormalizedRotationAxis=@n4
   PivotPoint=@n5
   Position=@n9
   Normal
   TangentU
   TangentV
  ) {materialFunction=/Game/Characters/Player/Male/Medium/Bodies/M_MED_Inferno/FX/Mat/RotateVertexAboutAxis}
  n11 = MaterialFunctionCall(RotationAngle=Multiply(A=@n1, B=@n2), NormalizedRotationAxis=@n4, PivotPoint=@n5, Position=@n9) {materialFunction=&RotateVertexAboutAxis}
  LinearInterpolate(
   B=LinearInterpolate(A=Add(A=@n10.WPO, B=@n9), B=Add(A=@n11.WPO, B=@n9), Alpha=VertexColor().R)
   Alpha=ScalarParameter() {defaultValue=1, group=ParticleType, parameterName=isBiFold, sortPriority=0}
  )
  n12 = VertexNormalWS()
  n13 = CameraVectorWS()
  n14 = ParticleRandom()
  n15 = Power(
   Base=Abs(None=DotProduct(A=@n12, B=@n13))
   Exp=LinearInterpolate(
    A=ScalarParameter() {defaultValue=3, parameterName=Iridescence Power}
    B=ScalarParameter() {defaultValue=0.1, parameterName=Iridescence Power2}
    Alpha=@n14
   )
  )
  n16 = TextureCoordinate() {uTiling=0.2, vTiling=0.2}
  n17 = Multiply(
   A=ComponentMask(None=Panner(Coordinate=@n16) {speedX=0.2, speedY=0.2}) {r=true}
   B=LinearInterpolate(A=ScalarParameter() {defaultValue=0.1, parameterName=Tiling}, B=ScalarParameter() {defaultValue=1, parameterName=Tiling2}, Alpha=@n14)
  )
  n18 = LinearInterpolate(A=@n17, B=Add(A=@n17, B=@n15), Alpha=Power(Base=@n15) {constExponent=1})
  n19 = Add(A=ParticleRandom(), B=Multiply(A=Time(), B=ScalarParameter() {defaultValue=0.5, parameterName=HuePanSpeed}))
  n20 = Constant3Vector() {constant=<rgba 1 0 0 0>}
  n21 = LinearInterpolate(A=@n20, B=Constant3Vector() {constant=<rgba 0 0 1 0>}, Alpha=VertexColor().R)
  n22 = ScalarParameter() {defaultValue=1, group=ParticleType, parameterName=isBiFold, sortPriority=0}
  n23 = Multiply(
   A=LinearInterpolate(B=ScalarParameter() {defaultValue=5, parameterName=AngleBrightness}, Alpha=@n15) {constA=0.05}
   B=MaterialFunctionCall(Hue Shift Percentage=Add(A=@n18, B=@n19), Texture=LinearInterpolate(A=@n20, B=@n21, Alpha=@n22)) {materialFunction=&HueShift}.Result
  )
  n24 = ParticleColor()
  n25 = ScalarParameter() {parameterName=ParticleColor_Presence}
  MaterialFunctionCall(Emissive=Multiply(A=NamedRerouteUsage(), B=LinearInterpolate(A=@n23, B=@n24.RGB, Alpha=@n25))) {materialFunction=&MF_FX_EmissiveAdjustments}
  n26 = TransformPosition(None=WorldPosition().XYZ) {transformSourceType=TRANSFORMPOSSOURCE_World, transformType=TRANSFORMPOSSOURCE_View}
  n27 = VertexNormalWS()
  n28 = MaterialFunctionCall(Color=CrossProduct(A=Normalize(VectorInput=@n26), B=Transform(None=@n27) {transformSourceType=TRANSFORMSOURCE_World, transformType=TRANSFORM_View})) {materialFunction=&SplitComponents}
  n29 = Add(A=Multiply(
    A=MaterialFunctionCall(R=@n28.G, G=@n28.G) {materialFunction=&AppendMany}.RG
    B=Constant2Vector() {g=0.5, r=0.5}
   )) {constB=0.5}
  n30 = ScalarParameter() {defaultValue=1, parameterName=FakeReflect_Scale}
  NamedRerouteDeclaration(None=TextureSample(UVs=Multiply(A=@n29, B=@n30)) {texture=/SparksEffects/Effects/Textures/Reflection/T_FakeCartoonGlassReflection}.R) {name=FakeReflectMask, nodeColor=<rgba 1 0 0.5391 1>}
  ScalarParameter() {defaultValue=0.5}
 ]
}
