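# Translucent, unlit, two-sided Niagara material for tech confetti on sprite and mesh particles, using a channel-packed triangle flipbook indexed per particle or manually.
# Emissive combines noise-animated, per-particle hue shift, view-angle brightness and a matcap-style fake glass reflection from view-space normals, then applies override color and brightness.
# Opacity combines sprite channels, reflection and particle alpha with a sharpened noise dissolve.
# Dynamic parameters drive dissolve, sprite index and hue offset.
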Material /SparksEffects/Effects/Materials/Digital/M_FX_TechConfetti_Sprites
graph {
 properties {
  bAutomaticallySetUsageInEditor=false
  bUsedWithNiagaraMeshParticles=true
  bUsedWithNiagaraSprites=true
  blendMode=BLEND_Translucent
  shadingModel=MSM_Unlit
  twoSided=true
  usedShadingModels=Unlit
 }
 nodes [
  n1 = Multiply(A=NamedRerouteUsage()) {constB=3}
  n2 = ScalarParameter() {defaultValue=1, group=4. Fake Reflection, parameterName=_FakeReflect_Presence}
  n3 = LinearInterpolate(B=NamedRerouteUsage(), Alpha=Power(Base=NamedRerouteUsage()) {constExponent=0.5})
  n4 = ScalarParameter() {defaultValue=10, group=3. View Angle Effect, parameterName=AngleBrightness}
  n5 = Power(Base=NamedRerouteUsage(), Exp=ScalarParameter() {defaultValue=3, group=3. View Angle Effect, parameterName=AngleDarkening_Pow})
  n6 = NamedRerouteUsage()
  n7 = ScalarParameter() {defaultValue=1, group=4. Fake Reflection, parameterName=_FakeReflect_Presence}
  n8 = Multiply(A=Multiply(A=LinearInterpolate(B=@n4, Alpha=@n5) {constA=0.1}, B=LinearInterpolate(B=@n6, Alpha=@n7) {constA=1}), B=ParticleColor().RGB)
  n9 = Multiply(A=Add(A=LinearInterpolate(B=@n1, Alpha=@n2), B=Add(A=@n3, B=@n8)), B=NamedRerouteUsage())
  n10 = VectorParameter() {group=0. Master Controls, parameterName=Override Color, defaultValue=<rgba 0.25 0 0.06 1>}
  n11 = ScalarParameter() {defaultValue=0.5, group=0. Master Controls, parameterName=_OverrideColor_Presence}
  MaterialFunctionCall(Emissive=Multiply(A=LinearInterpolate(A=@n9, B=@n10.RGB, Alpha=@n11), B=ScalarParameter() {defaultValue=2, group=0. Master Controls, parameterName=Brightness})) {materialFunction=&MF_FX_EmissiveAdjustments}
  n12 = VertexNormalWS()
  n13 = CameraVectorWS()
  NamedRerouteDeclaration(None=Abs(None=DotProduct(A=@n12, B=@n13))) {name=AngleDarkening, nodeColor=<rgba 0.9922 0 1 1>}
  NamedRerouteDeclaration(None=MaterialFunctionCall(
    "Animation  Phase (0-1)"=NamedRerouteUsage()
    Number of Rows=ScalarParameter() {defaultValue=2, group=1. Sprite Sheet, parameterName=SpriteSheet_Rows}
    Number of Columns=ScalarParameter() {defaultValue=2, group=1. Sprite Sheet, parameterName=SpriteSheet_Columns}
    Texture=TextureObjectParameter() {group=1. Sprite Sheet, parameterName=SpriteSheet, texture=/SparksEffects/Effects/Textures/Digital/T_TriangleSpriteBook}
   ) {materialFunction=&FlipBook}.Result) {name=SpriteSheet_Output, nodeColor=<rgba 0 0.1328 1 1>}
  n14 = TransformPosition(None=WorldPosition().XYZ) {transformSourceType=TRANSFORMPOSSOURCE_World, transformType=TRANSFORMPOSSOURCE_View}
  n15 = VertexNormalWS()
  n16 = MaterialFunctionCall(Color=CrossProduct(A=Normalize(VectorInput=@n14), B=Transform(None=@n15) {transformSourceType=TRANSFORMSOURCE_World, transformType=TRANSFORM_View})) {materialFunction=&SplitComponents}
  n17 = Add(A=Multiply(
    A=MaterialFunctionCall(R=@n16.G, G=@n16.G) {materialFunction=&AppendMany}.RG
    B=Constant2Vector() {g=0.5, r=0.5}
   )) {constB=0.5}
  n18 = ScalarParameter() {defaultValue=0.75, group=4. Fake Reflection, parameterName=FakeReflect_Scale}
  NamedRerouteDeclaration(None=TextureSample(UVs=Multiply(A=@n17, B=@n18)) {texture=/SparksEffects/Effects/Textures/Reflection/T_FakeCartoonGlassReflection}.R) {name=FakeReflectMask, nodeColor=<rgba 1 0 0.5391 1>}
  n19 = NamedRerouteUsage()
  n20 = NamedRerouteUsage()
  NamedRerouteDeclaration(None=LinearInterpolate(B=OneMinus(None=@n19), Alpha=ComponentMask(None=@n20) {g=true})) {name=WireframeMask, nodeColor=<rgba 0.5469 1 0 1>}
  n21 = ScalarParameter() {
   desc="Becomes available when isParticle == 0. Enter Value between 0 and [total number of sprite indices]."
   group=1. Sprite Sheet
   parameterName=ManualSpriteIndex
  }
  n22 = Multiply(
   A=ScalarParameter() {defaultValue=2, group=1. Sprite Sheet, parameterName=SpriteSheet_Rows}
   B=ScalarParameter() {defaultValue=2, group=1. Sprite Sheet, parameterName=SpriteSheet_Columns}
  )
  n23 = Floor(None=Multiply(A=@n21, B=@n22) {constB=4})
  n24 = Round(None=ScalarParameter() {defaultValue=1, group=1. Sprite Sheet, parameterName=isParticle})
  NamedRerouteDeclaration(None=LinearInterpolate(
    A=Divide(A=@n23, B=@n22) {constB=4}
    B=DynamicParameter() {paramNames [Dissolve, SpriteIndex0-1, HueOffset, Param4]}.SpriteIndex0-1
    Alpha=Saturate(None=@n24)
   )) {name=SpriteSheet_Index, nodeColor=<rgba 0.0078 0 1 1>}
  n25 = Multiply(A=NamedRerouteUsage(), B=ScalarParameter() {defaultValue=1, group=2. Hue Shift, parameterName=PerParticle_HueOffset})
  n26 = TextureCoordinate()
  n27 = ScalarParameter() {defaultValue=2, group=2. Hue Shift, parameterName=HueShiftNoiseTiling}
  n28 = Add(
   A=Multiply(A=NamedRerouteUsage()) {constB=7.77}
   B=Panner(Coordinate=MaterialFunctionCall(UVs=@n26, Texture Scale=@n27) {materialFunction=&ScaleUVsByCenter}.UVs) {speedX=0.1, speedY=0.05}
  )
  n29 = NamedRerouteUsage()
  n30 = ScalarParameter() {defaultValue=0.6, group=2. Hue Shift, parameterName=_HueShiftRange}
  n31 = NamedRerouteUsage()
  n32 = Panner(
   Coordinate=Multiply(
    A=TextureSample(UVs=@n28, Tex=@n29) {samplerType=SAMPLERTYPE_Grayscale, texture=&LowResBlurredNoise}.R
    B=LinearInterpolate(B=@n30, Alpha=@n31) {constA=0.05}
   ) {constB=0.5}
   Speed=ScalarParameter() {defaultValue=0.3, group=2. Hue Shift, parameterName=HueShiftSpeed}
  ) {speedX=0.1}
  n33 = Add(
   A=VertexColor() {desc=Vertex Color Red Channel drives additional Hue Offset}.R
   B=Add(
    A=DynamicParameter() {desc="Set to RandomRange in NS, if desired", paramNames [Dissolve, SpriteIndex0-1, HueOffset, Param4]}.HueOffset
    B=Add(A=@n25, B=@n32)
   )
  )
  n34 = Constant3Vector() {constant=<rgba 1 0 0 0>}
  NamedRerouteDeclaration(None=LinearInterpolate(
    A=ParticleColor().RGB
    B=MaterialFunctionCall(Hue Shift Percentage=@n33, Texture=@n34) {materialFunction=&HueShift}.Result
    Alpha=ScalarParameter() {defaultValue=1, group=2. Hue Shift, parameterName=_HueTexture_Presence}
   )) {name=Hue, nodeColor=<rgba 1 0.1413 0 1>}
  NamedRerouteDeclaration(None=TextureObject() {samplerType=SAMPLERTYPE_Grayscale, texture=&LowResBlurredNoise}) {name=NoiseTextureObject, nodeColor=<rgba 0.4531 0 1 1>}
  n35 = LinearInterpolate(B=NamedRerouteUsage(), Alpha=NamedRerouteUsage())
  n36 = LinearInterpolate(B=ComponentMask(None=NamedRerouteUsage()) {b=true}, Alpha=LinearInterpolate(Alpha=NamedRerouteUsage())) {constA=1}
  n37 = NamedRerouteUsage()
  n38 = ScalarParameter() {defaultValue=1, group=4. Fake Reflection, parameterName=_FakeReflect_Presence}
  n39 = Multiply(A=ParticleColor().A, B=Add(A=Add(A=@n35, B=@n36), B=LinearInterpolate(B=@n37, Alpha=@n38)))
  n40 = ComponentMask(None=NamedRerouteUsage()) {desc=Triangle alpha, r=true}
  n41 = TextureCoordinate()
  n42 = ScalarParameter() {defaultValue=0.8, group=5. Dissolve, parameterName=DissolveNoiseTiling}
  n43 = Add(
   A=Multiply(A=NamedRerouteUsage()) {constB=7.77}
   B=Panner(Coordinate=MaterialFunctionCall(UVs=@n41, Texture Scale=@n42) {materialFunction=&ScaleUVsByCenter}.UVs) {speedX=0.1, speedY=0.05}
  )
  n44 = NamedRerouteUsage()
  n45 = ScalarParameter() {defaultValue=10, group=5. Dissolve, parameterName=DissolveSharpness}
  n46 = DynamicParameter() {paramNames [Dissolve, SpriteIndex0-1, HueOffset, Param4]}
  n47 = Subtract(
   A=Multiply(
    A=TextureSample(UVs=@n43, Tex=@n44) {samplerType=SAMPLERTYPE_Grayscale, texture=&LowResBlurredNoise}.R
    B=@n45
   ) {constB=5}
   B=LinearInterpolate(A=Multiply(A=@n45) {constB=-1}, B=@n45, Alpha=OneMinus(None=@n46.Dissolve)) {constA=-5, constB=5}
  ) {constB=-1}
  Saturate(None=Multiply(A=Multiply(A=@n39, B=@n40), B=Saturate(None=@n47)))
 ]
}
